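# Material function for the RL196SE wheel's holographic look, layered over MF_WheelParent_Simple with emissive, opacity and dark metallic surface overrides.
# An RGB mask texture, with a switch for mirrored text, is tinted by two rim color parameters to form the star pattern.
# Time-driven cosine of radial UV distance, sharpened by a power, creates pulsing rings.
# Named reroutes carry these layers and vertex-color red; opacity optionally multiplies by an LOD mask.
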MaterialFunction /VehicleCosmetics/Wheels/RL196SE/Materials/MF_Wheel_RL196SE_HoloGraphic
graph {
 properties {previewBlendMode=BLEND_Translucent}
 nodes [
  n1 = Multiply(A=NamedRerouteUsage())
  n2 = Multiply(A=NamedRerouteUsage())
  n3 = NamedRerouteUsage()
  n4 = Multiply(A=NamedRerouteUsage(), B=Add(A=@n3, B=Multiply(A=@n3))) {desc=4}
  Add(B=Add(A=Add(A=@n1, B=@n2) {desc=3}, B=@n4) {desc=2}) {desc=1}
  n5 = NamedRerouteUsage()
  Clamp(None=@n5)
  Constant() {r=100}
  n6 = NamedRerouteUsage()
  n7 = Multiply(A=@n6, B=NamedRerouteUsage())
  n8 = TextureSampleParameter2D() {
   group=1_Input Textures
   parameterName=LOD_mask
   samplerType=SAMPLERTYPE_Grayscale
   texture=/VehicleCosmetics/Wheels/RL196SE/Textures/T_rim_rl196_SE_LODmask
  }
  n9 = Constant() {r=1}
  n10 = SetMaterialAttributes!(
   MaterialAttributes=MaterialFunctionCall() {materialFunction=&MF_WheelParent_Simple}.MaterialAttributes
   EmissiveColor=@n5
   BaseColor=Constant()
   Metallic=Constant() {r=1}
   Normal
   Specular
   Roughness=Constant()
   Opacity=Multiply(A=Add(A=@n6, B=@n7), B=StaticSwitchParameter(True=@n8.R, False=@n9) {group=1_Input Textures, parameterName=Use_LOD_mask})
  ) {attributeSetTypes [
    #1
    #2
    #3
    #4
    #5
    #6
    #7
   ]}
  FunctionOutput(None=@n10)
  n11 = Saturate(None=VectorParameter() {parameterName=Rim_ColorC, defaultValue=<rgba 0.1195 0.7011 0.8879 1>}.RGB)
  NamedRerouteDeclaration(None=@n11) {name=RimcolorC, nodeColor=<rgba 0 0.2031 1 1>}
  n12 = TextureSample() {texture=/VehicleCosmetics/Wheels/RL196SE/Textures/T_Rim_rl196_SE_RGB}
  n13 = StaticSwitchParameter(
   True=TextureSample(UVs=TextureCoordinate() {uTiling=-1}) {texture=&T_Rim_rl196_SE_RGB}.RGB
   False=@n12.RGB
  ) {parameterName=MirroredText}
  n14 = ComponentMask(None=@n13) {b=true}
  n15 = Saturate(None=VectorParameter() {parameterName=Rim_ColorB, defaultValue=<rgba 0.7758 0.2874 0.0823 1>}.RGB)
  n16 = ComponentMask(None=@n13) {g=true}
  n17 = Multiply(A=@n15, B=@n16)
  n18 = Add(A=Multiply(A=@n11, B=Add(A=@n12.R, B=@n14)), B=@n17)
  n19 = NamedRerouteUsage()
  n20 = Multiply(A=Add(A=@n18, B=@n17), B=NamedRerouteUsage())
  n21 = Add(A=@n20, B=@n20)
  n22 = Add(A=@n21, B=@n21)
  n23 = Clamp(None=Multiply(A=Add(A=@n22, B=@n22)) {constB=2}) {maxDefault=2.5}
  NamedRerouteDeclaration(None=Add(A=@n18, B=Multiply(A=@n19, B=@n23))) {name=StaticStar, nodeColor=<rgba 0 1 0.8359 1>}
  NamedRerouteDeclaration(None=Add(A=ComponentMask(None=@n13) {r=true}, B=@n16)) {name=StaticStar_combined, nodeColor=<rgba 1 0.5156 0 1>}
  n24 = Multiply(
   A=Subtract(A=TextureCoordinate()) {constB=0.5}
   B=ScalarParameter() {defaultValue=0.25, group=00_Base, parameterName=Rings Bands Amount, sortPriority=14}
  )
  n25 = Constant2Vector()
  n26 = Time()
  n27 = ScalarParameter() {defaultValue=-0.05, group=00_Base, parameterName=Rings Pulse Speed, sortPriority=15}
  n28 = Cosine(None=Add(A=Distance(A=@n24, B=@n25), B=Multiply(A=@n26, B=@n27) {constB=0.025})) {period=0.075}
  n29 = Power(Base=Divide(A=Add(A=@n28))) {constExponent=60}
  NamedRerouteDeclaration(None=Multiply(A=Clamp(None=@n29)) {constB=5}) {name=EmissiveRings, nodeColor=<rgba 0 0.9062 1 1>}
  NamedRerouteDeclaration(None=@n15) {name=Rimcolor_b, nodeColor=<rgba 0.1953 0 1 1>}
  NamedRerouteDeclaration(None=@n4) {name=Emissive_out, nodeColor=<rgba 0 0.7656 1 1>}
  NamedRerouteDeclaration(None=VertexColor().R) {name=VC_R, nodeColor=<rgba 1 0 0 1>}
  NamedRerouteUsage()
 ]
}
