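# Material function for the Inverted RL196SE wheel rim, overriding MF_WheelParent_Simple attributes with a black, smooth metal surface and an emissive input.
# The emissive graph combines counter-rotating textures under a circular gradient, tinted star channels, MF_Infinity_OffsetFunction depth layers and a sine-driven pulse.
# Static switches select mirrored U sampling for part of the star art and an optional LOD opacity mask.
# A time-based Frac animation of depth and darkness is left unconnected.
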MaterialFunction /VehicleCosmetics/Wheels/RL196SE/Materials/MF_Wheel_RL196SE_Inverted
graph {nodes [
  n1 = TextureSampleParameter2D() {
   group=1_Input Textures
   parameterName=LOD_mask
   samplerType=SAMPLERTYPE_Grayscale
   texture=/VehicleCosmetics/Wheels/RL196SE/Textures/T_rim_rl196_SE_LODmask
  }
  n2 = SetMaterialAttributes!(
   MaterialAttributes=MaterialFunctionCall() {materialFunction=&MF_WheelParent_Simple}.MaterialAttributes
   EmissiveColor=NamedRerouteUsage()
   BaseColor=Constant()
   Metallic=Constant() {r=1}
   Normal
   Specular
   Roughness=Constant()
   OpacityMask=StaticSwitchParameter(True=@n1.R, False=Constant() {r=1}) {group=1_Input Textures, parameterName=Use_LOD_mask}
  ) {attributeSetTypes [
    #1
    #2
    #3
    #4
    #5
    #6
    #7
   ]}
  FunctionOutput(None=@n2)
  n3 = VectorParameter() {parameterName=Rim_ColorC, defaultValue=<rgba 0.1195 0.7011 0.8879 1>}
  NamedRerouteDeclaration(None=@n3.RGB) {name=RimcolorC, nodeColor=<rgba 0 0.2031 1 1>}
  n4 = TextureSample() {texture=/VehicleCosmetics/Wheels/RL196SE/Textures/T_Rim_rl196_SE_RGB}
  n5 = VectorParameter() {parameterName=Rim_ColorB, defaultValue=<rgba 0.7758 0.2874 0.0823 1>}
  n6 = StaticSwitchParameter(
   True=TextureSample(UVs=TextureCoordinate() {uTiling=-1}) {texture=&T_Rim_rl196_SE_RGB}.RGB
   False=@n4.RGB
  ) {parameterName=MirroredText}
  n7 = ComponentMask(None=@n6) {g=true}
  NamedRerouteDeclaration(None=Add(A=Multiply(A=@n3.RGB, B=@n4.R), B=Multiply(A=@n5.RGB, B=@n7))) {name=StaticStar, nodeColor=<rgba 0 1 0.8359 1>}
  NamedRerouteDeclaration(None=Add(A=ComponentMask(None=@n6) {r=true}, B=@n7)) {name=StaticStar_combined, nodeColor=<rgba 1 0.5156 0 1>}
  NamedRerouteDeclaration(None=@n5.RGB) {name=Rimcolor_b, nodeColor=<rgba 0.1953 0 1 1>}
  n8 = ScalarParameter() {defaultValue=0.3, parameterName=Infinity_Strength}
  n9 = ScalarParameter() {defaultValue=-1, parameterName=Infinity_Maxdepth}
  n10 = ScalarParameter() {defaultValue=0.005, parameterName=GradientDarkness_infinity}
  n11 = Multiply(
   A=Constant() {r=1}
   B=MaterialFunctionCall!(
    Texture=TextureObject() {texture=&T_Rim_rl196_SE_RGB}
    Offset=@n9
    RotateUVs
    GradientDarkness=@n10
    GradientFalloff
    UVs
    Opt_LowQualityOffset
    Opt_MediumQualityOffset
   ) {materialFunction=/VehicleCosmetics/SharedMaterials/MaterialFunctions/MF_Infinity_OffsetFunction}.Result
  )
  NamedRerouteDeclaration(None=Multiply(A=@n8, B=ComponentMask(None=@n11) {g=true})) {name=Inf_gg, nodeColor=<rgba 0.125 1 0 1>}
  NamedRerouteDeclaration(None=Multiply(A=@n8, B=ComponentMask(None=@n11) {r=true})) {name=Inf_star, nodeColor=<rgba 0.125 1 0 1>}
  n12 = TextureSample(UVs=Rotator() {speed=0.5}) {texture=&T_Rim_rl196_SE_RGB}
  n13 = TextureSample(UVs=Rotator() {speed=-0.5}) {texture=&T_Rim_rl196_SE_RGB}
  n14 = Multiply(A=ScalarParameter() {defaultValue=0.5, parameterName=BackGroundIntensity}, B=Multiply(A=NamedRerouteUsage(), B=Add(A=@n12.R, B=@n13.R)))
  n15 = Power(Base=TextureSample() {samplerType=SAMPLERTYPE_Grayscale, texture=/VehicleCosmetics/Wheels/RL196SE/Textures/T_GradientCircle01}.R) {constExponent=1.3}
  n16 = Add(A=Multiply(A=NamedRerouteUsage(), B=NamedRerouteUsage()), B=Multiply(A=NamedRerouteUsage(), B=NamedRerouteUsage())) {desc=3}
  n17 = NamedRerouteUsage()
  n18 = Multiply(A=Time(), B=ScalarParameter() {defaultValue=0.5, parameterName=speed})
  n19 = Divide(A=Add(A=Sine(None=@n18)))
  n20 = Multiply(A=NamedRerouteUsage(), B=Add(A=@n17, B=Multiply(A=@n17, B=@n19))) {desc=4}
  NamedRerouteDeclaration(None=Add(A=Multiply(A=@n14, B=@n15), B=Add(A=@n16, B=@n20) {desc=2}) {desc=1}) {name=Emissive_out, nodeColor=<rgba 0 0.7656 1 1>}
  NamedRerouteDeclaration(None=VertexColor().R) {name=VC_R, nodeColor=<rgba 1 0 0 1>}
  n21 = Frac(None=Multiply(A=Time(), B=ScalarParameter() {defaultValue=0.5, parameterName=speed}))
  Reroute(None=Multiply(A=@n9, B=Subtract(B=@n21) {constA=0.2}))
  n22 = OneMinus(None=@n21)
  Reroute(None=Multiply(A=@n22, B=@n10))
  Reroute(None=@n22)
  VectorParameter() {parameterName=Rim_ColorA, defaultValue=<rgba 1 0.1816 0.8834 1>}
 ]}
