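# Holographic base material for the RL196 SE cosmetic wheel, with tunable custom and accent colors for skeletal meshes.
# Vertex color red splits rim from tire, blending both the colors and the normal maps.
# Clamped emissive multiplies those colors, two counter-panning cloud layers, a rim mask and layered Fresnel terms, including a very sharp facing highlight.
# Two slowly panning glitter textures perturb the normals, which likely feed the Fresnel for a shifting sparkle.
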Material /VehicleCosmetics/Wheels/RL196SE/Materials/M_wheel_196SE_Holo_base
graph {
 properties {bAutomaticallySetUsageInEditor=false, bUsedWithSkeletalMesh=true, usedShadingModels=Default Lit}
 nodes [
  n1 = TextureCoordinate()
  n2 = TextureCoordinate() {uTiling=2, vTiling=2}
  n3 = Add(
   A=TextureSample(UVs=Panner(Coordinate=@n1) {speedX=0.33, speedY=0.1}) {samplerType=SAMPLERTYPE_Grayscale, texture=/VehicleCosmetics/Wheels/RL273SE/Textures/T_Wheel_RL273SE_clouds}.R
   B=TextureSample(UVs=Panner(Coordinate=@n2) {speedX=-0.23, speedY=-0.18}) {samplerType=SAMPLERTYPE_Grayscale, texture=&T_Wheel_RL273SE_clouds}.R
  )
  n4 = VertexColor()
  n5 = LinearInterpolate(
   A=VectorParameter() {parameterName=CustomColor, defaultValue=<rgba 0.3494 0.5863 0.5938 1>}.RGB
   B=Multiply(
    A=VectorParameter() {parameterName=AccentColor, defaultValue=<rgba 1 0.7072 0.5 1>}.RGB
    B=ScalarParameter() {defaultValue=1, parameterName=DarkColorMultiply}
   ) {constB=3}
   Alpha=@n4.R
  ) {desc=AO}
  n6 = Multiply(A=Clamp(None=Divide(A=@n3)) {minDefault=0.3}, B=@n5)
  n7 = TextureSample() {samplerType=SAMPLERTYPE_LinearColor, texture=/VehicleCosmetics/Wheels/RL196/Textures/T_Rim_RL196_M}
  n8 = ScalarParameter() {defaultValue=1, parameterName=FresnelStrength}
  n9 = Transform(None=NamedRerouteUsage())
  n10 = Fresnel(Normal=@n9) {exponent=3}
  n11 = ConstantBiasScale(None=Power(Base=OneMinus(None=@n10)) {constExponent=200}) {bias=0, scale=2000}
  n12 = Add(A=Power(Base=Fresnel(Normal=@n9) {exponent=3}) {constExponent=1}, B=Add(A=Clamp(None=@n11), B=Fresnel(Normal=@n9) {exponent=3}))
  n13 = Multiply(A=Multiply(A=LinearInterpolate(A=@n7.A, Alpha=@n4.R) {desc=AO}, B=@n5), B=Add(B=Multiply(A=@n8, B=@n12)) {constA=0.2})
  Clamp(None=Multiply(A=ScalarParameter() {defaultValue=1.5, parameterName=EmissiveStrength}, B=Multiply(A=@n6, B=@n13))) {maxDefault=5}
  n14 = Multiply(A=Time()) {constB=0.01}
  n15 = TextureSample(UVs=Panner(Coordinate=TextureCoordinate() {uTiling=0.03, vTiling=0.06}, Time=@n14) {speedX=0.33, speedY=0.1}) {samplerType=SAMPLERTYPE_LinearColor, texture=/VehicleCosmetics/Wheels/RL196SE/Textures/T_glitter}
  n16 = TextureSample(UVs=Panner(Coordinate=TextureCoordinate() {uTiling=0.05, vTiling=0.05}, Time=@n14) {speedX=-0.23, speedY=-0.18}) {samplerType=SAMPLERTYPE_LinearColor, texture=&T_glitter}
  NamedRerouteDeclaration(None=Multiply(A=Multiply(A=@n15.RGB, B=@n16.RGB), B=ScalarParameter() {defaultValue=0.8, parameterName=Squares_amount})) {name=Squares, nodeColor=<rgba 0.0312 1 0 1>}
  n17 = NamedRerouteUsage()
  n18 = TextureSampleParameter2D() {parameterName=Rim_Normal, samplerType=SAMPLERTYPE_LinearColor, texture=/VehicleCosmetics/Wheels/RL196/Textures/T_Rim_RL196_N}
  n19 = Add(
   A=TextureSampleParameter2D() {parameterName=Tire_Normal, samplerType=SAMPLERTYPE_LinearColor, texture=/VehicleCosmetics/Wheels/RL196/Textures/T_Tire_RL196_N}.RGB
   B=TextureSample() {texture=/VehicleCosmetics/Wheels/RL196SE/Textures/T_tire_rl196_SE_RGB}.R
  )
  n20 = LinearInterpolate(A=Normalize(VectorInput=Add(A=@n17, B=@n18.RGB)), B=Normalize(VectorInput=Add(A=@n17, B=@n19)), Alpha=@n4.R)
  NamedRerouteDeclaration(None=Add(A=@n20, B=@n20)) {name=Normal_out, nodeColor=<rgba 1 0 0.3516 1>}
  NamedRerouteUsage()
 ]
}
